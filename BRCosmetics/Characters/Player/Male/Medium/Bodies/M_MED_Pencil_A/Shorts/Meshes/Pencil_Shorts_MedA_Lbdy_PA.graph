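# Physics asset for the left leg of the Pencil medium male character's shorts, simulating secondary motion of the fabric.
# Capsules on the pelvis, left thigh twist and calf form the body.
# Four tiny master spheres (front, back, left, right) attach to the thigh with free rotation.
# Each master carries a capsule dynamic bone with mostly limited swing and tight twist, with the back one narrower on one swing axis.
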PhysicsAsset /BRCosmetics/Characters/Player/Male/Medium/Bodies/M_MED_Pencil_A/Shorts/Meshes/Pencil_Shorts_MedA_Lbdy_PA
bodies {
 calf_l [{shapeType=Capsule, length=16, radius=5, center=<-16 -1 0>, rotation=<rot 90 0 0>}]
 pelvis [{shapeType=Capsule, length=20, radius=5, center=<3.5 0 0>}]
 shorts_l_leg_back_dyn [{shapeType=Capsule, length=10, radius=8, shapeName=shorts_l_leg_back_dyn_capsule, center=<5.5 0 9>, rotation=<rot 0 0 90>}]
 shorts_l_leg_back_mstr [{shapeType=Sphere, radius=0.505, shapeName=shorts_l_leg_back_mstr_sphere}]
 shorts_l_leg_front_dyn [{shapeType=Capsule, length=10, radius=8, shapeName=shorts_l_leg_front_dyn_capsule, center=<5.5 0 8.5>, rotation=<rot 0 0 90>}]
 shorts_l_leg_front_mstr [{shapeType=Sphere, radius=0.505, shapeName=shorts_l_leg_front_mstr_sphere}]
 shorts_l_leg_left_dyn [{shapeType=Capsule, length=10, radius=8, shapeName=shorts_l_leg_left_dyn_capsule, center=<5.5 0 9.5>, rotation=<rot 0 0 90>}]
 shorts_l_leg_left_mstr [{shapeType=Sphere, radius=0.505, shapeName=shorts_l_leg_left_mstr_sphere}]
 shorts_l_leg_right_dyn [{shapeType=Capsule, length=10, radius=8, shapeName=shorts_l_leg_right_dyn_capsule, center=<5.5 0 9.5>, rotation=<rot 0 0 90>}]
 shorts_l_leg_right_mstr [{shapeType=Sphere, radius=0.505, shapeName=shorts_l_leg_right_mstr_sphere}]
 thigh_twist_01_l [{shapeType=Capsule, length=15.5, radius=7.5, center=<-5 -0.5 0.5>, rotation=<rot 90 0 0>}]
}
constraints [
 {bone1Name=thigh_twist_01_l, bone2Name=pelvis, swing1LimitDegrees=45, swing2LimitDegrees=45, twistLimitDegrees=45}
 {
  bone1Name=shorts_l_leg_back_mstr
  bone2Name=thigh_twist_01_l
  swing1LimitDegrees=45
  swing1Motion=Free
  swing2LimitDegrees=45
  swing2Motion=Free
  twistLimitDegrees=45
  twistMotion=Free
 }
 {bone1Name=shorts_l_leg_back_dyn, bone2Name=shorts_l_leg_back_mstr, swing1LimitDegrees=45, swing2LimitDegrees=15, twistLimitDegrees=5}
 {
  bone1Name=shorts_l_leg_front_mstr
  bone2Name=thigh_twist_01_l
  swing1LimitDegrees=45
  swing1Motion=Free
  swing2LimitDegrees=45
  swing2Motion=Free
  twistLimitDegrees=45
  twistMotion=Free
 }
 {bone1Name=shorts_l_leg_front_dyn, bone2Name=shorts_l_leg_front_mstr, swing1LimitDegrees=45, swing2LimitDegrees=45, twistLimitDegrees=5}
 {
  bone1Name=shorts_l_leg_left_mstr
  bone2Name=thigh_twist_01_l
  swing1LimitDegrees=45
  swing1Motion=Free
  swing2LimitDegrees=45
  swing2Motion=Free
  twistLimitDegrees=45
  twistMotion=Free
 }
 {bone1Name=shorts_l_leg_left_dyn, bone2Name=shorts_l_leg_left_mstr, swing1LimitDegrees=45, swing2LimitDegrees=45, twistLimitDegrees=5}
 {
  bone1Name=shorts_l_leg_right_mstr
  bone2Name=thigh_twist_01_l
  swing1LimitDegrees=45
  swing1Motion=Free
  swing2LimitDegrees=45
  swing2Motion=Free
  twistLimitDegrees=45
  twistMotion=Free
 }
 {bone1Name=shorts_l_leg_right_dyn, bone2Name=shorts_l_leg_right_mstr, swing1LimitDegrees=45, swing2LimitDegrees=45, twistLimitDegrees=5}
 {
  bone1Name=calf_l
  bone2Name=pelvis
  swing1LimitDegrees=45
  swing1Motion=Free
  swing2LimitDegrees=45
  swing2Motion=Free
  twistLimitDegrees=45
  twistMotion=Free
 }
]
