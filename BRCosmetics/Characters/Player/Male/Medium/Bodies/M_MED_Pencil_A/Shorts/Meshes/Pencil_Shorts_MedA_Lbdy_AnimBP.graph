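# Post-attach animation Blueprint for the Pencil shorts lower-body cosmetic on medium male characters, copying the pose and curves from the attached parent mesh.
# The copied pose feeds a shorts Control Rig, whose result goes to the output.
# Each update derives crouch-moving, crouch-still and crouch alpha values for that rig from crouch and acceleration state, plus a deprecated gravity override.
# Chained per-thigh RigidBody nodes and a second deform rig sit unconnected to the output.
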AnimBlueprint /BRCosmetics/Characters/Player/Male/Medium/Bodies/M_MED_Pencil_A/Shorts/Meshes/Pencil_Shorts_MedA_Lbdy_AnimBP
target_skeleton=&/Pencil_Shorts_MedA_Lbdy_Skeleton
graphs [
 AnimGraph {nodes [
   Control Rig() {
    alphaBoolBlend {blendOption=Linear}
    controlRigAssetReference {blueprintRigClass=&/Pencil_Shorts_MedA_Lbdy_DeformRig_Master_ControlRig_new.Pencil_Shorts_MedA_Lbdy_DeformRig_Master_ControlRig_new_C}
   }
   Input Pose()
   n1 = RigidBody!(ComponentPose, Alpha, OverrideWorldGravity:Vector="0,0,-980", ExternalForce) -> (Pose) {
    bClampLinearTranslationLimitToRefPose=true
    bOverrideWorldGravity=true
    overridePhysicsAsset=&/Pencil_Shorts_MedA_Lbdy_PA
    simulationSpace=BaseBoneSpace
    alphaBoolBlend {blendOption=Linear}
    baseBoneRef {boneName=thigh_l}
    componentAppliedLinearAccClamp=<1000 1000 1000>
    componentLinearAccScale=<0.2 0.2 0.2>
    componentLinearVelScale=<0.2 0.2 0.2>
    overrideWorldGravity {z=-980}
    simSpaceSettings {
     maxAngularAcceleration=1000
     maxAngularVelocity=3
     maxLinearAcceleration=700
     maxLinearVelocity=500
     velocityScaleZ=0
     worldAlpha=0.5
     externalLinearDragV=<0.2 0.2>
    }
   }
   RigidBody(ComponentPose=@n1, OverrideWorldGravity="0,0,-980") {
    bClampLinearTranslationLimitToRefPose=true
    bOverrideWorldGravity=true
    overridePhysicsAsset=&/Pencil_Shorts_MedA_Lbdy_Right_PA
    simulationSpace=BaseBoneSpace
    alphaBoolBlend {blendOption=Linear}
    baseBoneRef {boneName=thigh_r}
    componentAppliedLinearAccClamp=<1000 1000 1000>
    componentLinearAccScale=<0.2 0.2 0.2>
    componentLinearVelScale=<0.2 0.2 0.2>
    overrideWorldGravity {z=-980}
    simSpaceSettings {
     maxAngularAcceleration=1000
     maxAngularVelocity=3
     maxLinearAcceleration=700
     maxLinearVelocity=500
     velocityScaleZ=0
     worldAlpha=0.5
     externalLinearDragV=<0.2 0.2>
    }
   }
   n2 = Control Rig!(
    Source=Copy Pose From Mesh() {bCopyCurves=true, bUseAttachedParent=true}
    Alpha
    IsCrouchMoving:bool=Get Crouch_Moving() -> (Crouch_Moving:bool)
    Crouch_Not_Moving:bool=Get Crouch Not Moving() -> (Crouch Not Moving:bool)
    Jacket_Crouch_Alpha:double=Get Jacket Crouch Alpha() -> (Jacket Crouch Alpha:double)
   ) -> (Pose) {
    alphaBoolBlend {blendOption=Linear}
    controlRigAssetReference {blueprintRigClass=&/Pencil_Shorts_MedA_Lbdy_ControlRig.Pencil_Shorts_MedA_Lbdy_ControlRig_C}
   }
   Output Pose(Result=Component To Local(ComponentPose=Local To Component(LocalPose=@n2))) {layerGroup=None}
  ]}
 EventGraph {nodes [
   Event BlueprintUpdateAnimation()
   > n1 = Sequence!(execute) -> (then_0, then_1, then_2:Exec, then_3:Exec)
   > n2 = Gravity Override State Function DEPRECATED(CharacterPartAnimInstance=Self-Reference(), GameTimeInSeconds=GetGameTimeInSeconds(), GravityOverrideParams=Get Gravity Override Params())
   > Set Gravity Override(Gravity Override=@n2)
   n3 = Get bIsAccelerating2D()
   n4 = Branch(execute=@n1.then_1, Condition=AND Boolean(A=Get bIsCrouching(), B="Not Equal (Boolean)"(A=true, B=@n3) -> (ReturnValue=~)))
   > Set Crouch Not Moving(Crouch Not Moving:bool=true) -> (Output_Get:bool)
   Set Crouch Not Moving(execute=@n4.else, Crouch Not Moving=false)
   Set Jacket Crouch Alpha(execute=@n1.then_2, Jacket Crouch Alpha:double=SelectFloat(A=1, B=0, bPickA=Get bIsCrouching())) -> (Output_Get:double)
   n5 = Branch(execute=@n1.then_3, Condition=AND Boolean(A=Get bIsCrouching(), B=Get bIsAccelerating2D()))
   > Set Crouch_Moving(Crouch_Moving:bool=true) -> (Output_Get:bool)
   Set Crouch_Moving(execute=@n5.else, Crouch_Moving=false)
  ]}
]
variables [
 Crouch Control {type=boolean, default=true}
 Crouch Not Moving {type=boolean}
 Crouch_Moving {type=boolean}
 Gravity Override {type=Vector}
 Gravity Override Params {type=GravityOverrideParamsStruct}
 Jacket Crouch Alpha {type=number}
]
